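# Post-process material from the Snow effects set that creates animated radial chromatic distortion around a world-space Location parameter.
# It projects Location to screen space and scrolls a reused lava mask in radial coordinates, sampled with screen-UV derivatives to avoid seam artifacts.
# A view-alignment mask and Offset Intensity scale that noise into separate red, green and blue scene-color UV offsets.
# A radial gradient mask branch is computed but does not affect the recombined color.
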Material /Figment_S10/Athena/Effects/Particles/Snow/Post/PPM_Snow_Final
graph {
 properties {blendMode=BLEND_Translucent, materialDomain=MD_PostProcess, pixelDepthOffsetMode=PDOM_Legacy, usedShadingModels=Default Lit}
 nodes [
  n1 = ScreenPosition()
  n2 = ScreenPosition()
  n3 = VectorParameter() {parameterName=Location, defaultValue=<rgba 66049 -43008 -2743 1>}
  n4 = MaterialFunctionCall(Vector or UVs=Subtract(
    A=@n2.ViewportUV
    B=MaterialFunctionCall(World Position=@n3.RGB) {materialFunction=&TransformToClipSpace}.Buffer UVs for Scene Textures
   )) {materialFunction=&VectorToRadialValue}
  n5 = Time()
  n6 = Add(A=@n4.Radial Coordinates, B=AppendVector(A=Constant(), B=Multiply(A=@n5) {constB=-0.5})) {constB=0}
  n7 = DDX(Value=@n2.ViewportUV)
  n8 = DDY(Value=@n2.ViewportUV)
  n9 = Multiply(A=Subtract(A=TextureSample!(UVs=@n6, Tex, "DDX(UVs)"=@n7, "DDY(UVs)"=@n8, Apply View MipBias) {mipValueMode=TMVM_Derivative, samplerType=SAMPLERTYPE_Masks, texture=/Game/Environments/Sets/Vulcano/Textures/Lava_M}.RGB)) {constB=1.5}
  n10 = CameraVectorWS()
  n11 = Normalize(VectorInput=Subtract(A=CameraPositionWS(), B=@n3.RGB))
  n12 = Saturate(None=Power(Base=DotProduct(A=@n10, B=@n11)) {constExponent=6})
  n13 = Multiply(A=ComponentMask(None=Multiply(A=@n9, B=@n12)) {g=true, r=true}, B=ScalarParameter() {defaultValue=0.0012, parameterName=Offset Intensity})
  n14 = Multiply(A=@n13) {constB=0.5}
  n15 = Add(A=@n1.ViewportUV, B=@n14)
  n16 = Add(A=@n1.ViewportUV, B=Multiply(A=@n14) {constB=-1})
  n17 = Add(A=@n1.ViewportUV, B=@n13)
  MaterialFunctionCall(
   X=ComponentMask(None=SceneTexture(UVs=@n15) {sceneTextureId=PPI_PostProcessInput0}.Color) {r=true}
   Y=ComponentMask(None=SceneTexture(UVs=@n16) {sceneTextureId=PPI_PostProcessInput0}.Color) {g=true}
   Z=ComponentMask(None=SceneTexture(UVs=@n17) {sceneTextureId=PPI_PostProcessInput0}.Color) {b=true}
  ) {materialFunction=&MakeFloat3}
  n18 = Saturate(None=MaterialFunctionCall(UV Channel=@n4.Radial Coordinates) {materialFunction=&LinearGradient}.VGradient)
  Multiply(A=Saturate(None=Multiply(A=@n18)), B=Multiply(A=@n18, B=@n12))
  OneMinus(None=@n18)
 ]
}
